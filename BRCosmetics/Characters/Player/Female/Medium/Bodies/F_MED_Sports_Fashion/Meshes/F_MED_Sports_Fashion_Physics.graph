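# Physics asset for the female medium Sports Fashion body cosmetic, setting up physics for two dangling string chains.
# Each string runs from a small box base through capsule segments, four numbered links on the right and three on the left.
# Swing and twist limited constraints join each link to the one above, with a tighter swing at the first numbered link.
# Pelvis and thigh capsules carry no constraints and provide body shapes around the strings.
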PhysicsAsset /BRCosmetics/Characters/Player/Female/Medium/Bodies/F_MED_Sports_Fashion/Meshes/F_MED_Sports_Fashion_Physics
bodies {
 dyn_string_l_01 [{shapeType=Capsule, length=3, radius=0.88, center=<2.4492 0 0>, rotation=<rot 90 0 0>}]
 dyn_string_l_02 [{shapeType=Capsule, length=3, radius=0.88, center=<2.4492 0 0>, rotation=<rot 90 0 0>}]
 dyn_string_l_03 [{shapeType=Capsule, length=5.3679, radius=0.88, center=<3.8711 0 0>, rotation=<rot 90 0 0>}]
 dyn_string_l_base [{shapeType=Box, extentX=0.5, extentY=0.5, extentZ=0.5}]
 dyn_string_l_parent [{shapeType=Capsule, length=2, radius=0.75, center=<1.5 0 0>, rotation=<rot 90 0 0>}]
 dyn_string_r_01 [{shapeType=Capsule, length=3, radius=0.88, center=<3 0 0>, rotation=<rot 90 0 0>}]
 dyn_string_r_02 [{shapeType=Capsule, length=3, radius=0.88, center=<3.6401 0 0>, rotation=<rot 90 0 0>}]
 dyn_string_r_03 [{shapeType=Capsule, length=3, radius=0.88, center=<3.2978 0 0>, rotation=<rot 90 0 0>}]
 dyn_string_r_04 [{shapeType=Capsule, length=3, radius=0.88, center=<2.4492 0 0>, rotation=<rot 90 0 0>}]
 dyn_string_r_base [{shapeType=Box, extentX=0.5, extentY=0.5, extentZ=0.5}]
 dyn_string_r_parent [{shapeType=Capsule, length=2, radius=0.75, center=<1.5 0 0>, rotation=<rot 90 0 0>}]
 pelvis [{shapeType=Capsule, length=10, radius=9.5, center=<0 -2.4413 0>}]
 thigh_l [{shapeType=Capsule, length=22.75, radius=7.25, center=<-15.0971 -0.3938 -0.325>, rotation=<rot 90 -3.5763 1.4205>}]
 thigh_r [{shapeType=Capsule, length=22.375, radius=7, center=<15.1699 0.0045 -0.4884>, rotation=<rot 90 1.7899 7.8297>}]
}
constraints [
 {bone1Name=dyn_string_r_parent, bone2Name=dyn_string_r_base, swing1LimitDegrees=60, swing2LimitDegrees=60, twistLimitDegrees=1}
 {bone1Name=dyn_string_r_01, bone2Name=dyn_string_r_parent, swing1LimitDegrees=30, swing2LimitDegrees=30, twistLimitDegrees=5}
 {bone1Name=dyn_string_r_02, bone2Name=dyn_string_r_01, swing1LimitDegrees=60, swing2LimitDegrees=60, twistLimitDegrees=10}
 {bone1Name=dyn_string_r_03, bone2Name=dyn_string_r_02, swing1LimitDegrees=60, swing2LimitDegrees=60, twistLimitDegrees=10}
 {bone1Name=dyn_string_r_04, bone2Name=dyn_string_r_03, swing1LimitDegrees=60, swing2LimitDegrees=60, twistLimitDegrees=10}
 {bone1Name=dyn_string_l_parent, bone2Name=dyn_string_l_base, swing1LimitDegrees=60, swing2LimitDegrees=60, twistLimitDegrees=1}
 {bone1Name=dyn_string_l_01, bone2Name=dyn_string_l_parent, swing1LimitDegrees=30, swing2LimitDegrees=30, twistLimitDegrees=5}
 {bone1Name=dyn_string_l_02, bone2Name=dyn_string_l_01, swing1LimitDegrees=60, swing2LimitDegrees=60, twistLimitDegrees=10}
 {bone1Name=dyn_string_l_03, bone2Name=dyn_string_l_02, swing1LimitDegrees=60, swing2LimitDegrees=60, twistLimitDegrees=10}
]
