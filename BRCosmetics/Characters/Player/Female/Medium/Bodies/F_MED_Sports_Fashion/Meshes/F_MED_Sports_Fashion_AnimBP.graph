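# Secondary-motion AnimBP for the F_MED_Sports_Fashion cosmetic body, adding dynamic pants and drawstring motion to the parent mesh's pose.
# A copied pose feeds a deform Control Rig, then low-alpha AnimDynamics on six pants bones and pelvis/spine constraints on the string bases.
# Skydive floating rotates the strings and raises the gravity-overridden RigidBody alpha from partial to full.
# Skydiving or surface swimming then offsets the outer pants bones.
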AnimBlueprint /BRCosmetics/Characters/Player/Female/Medium/Bodies/F_MED_Sports_Fashion/Meshes/F_MED_Sports_Fashion_AnimBP
target_skeleton=&/F_MED_Sports_Fashion_Skeleton
graphs [
 AnimGraph {nodes [
   n1 = Control Rig(Source=Copy Pose From Mesh() {bUseAttachedParent=true}) {
    alphaBoolBlend {blendOption=Linear}
    controlRigAssetReference {blueprintRigClass=&/F_MED_Sports_Fashion_DeformRig_Master_ControlRig.F_MED_Sports_Fashion_DeformRig_Master_ControlRig_C}
   }
   n2 = Anim Dynamics - Bone: dyn_pants_l_inner(ComponentPose=Local To Component(LocalPose=@n1), Alpha=1, ExternalForce="0,0,0") {
    angularDampingOverride=0.999
    angularSpringConstant=5000
    bAngularSpring=true
    bGravityOverrideInSimSpace=true
    bLinearSpring=true
    bOverrideAngularDamping=true
    gravityScale=0.75
    linearSpringConstant=5000
    simulationSpace=RootRelative
    alphaBoolBlend {blendOption=Linear}
    alphaScaleBiasClamp {scale=0.2}
    boundBone {boneName=dyn_pants_l_inner}
    componentAppliedLinearAccClamp=<6000 6000 6000>
    componentLinearAccScale=<0.25 0.25 0.6>
    componentLinearVelScale=<0.25 0.25 0.6>
    physicsBodyDefinitions [{
      sphereCollisionRadius=0
      boundBone {boneName=dyn_pants_l_inner}
      boxExtents=<8 25 25>
      constraintSetup {angularLimitsMax=<60 40 40>, angularLimitsMin=<-60 -40 -75>, angularTarget=<1 0 0>}
      localJointOffset=<5 0 0>
     }]
    relativeSpaceBone {boneName=root}
   }
   n3 = Anim Dynamics - Bone: dyn_pants_l_outer(ComponentPose=@n2, Alpha=1, ExternalForce="0,0,0") {
    angularDampingOverride=0.999
    angularSpringConstant=5000
    bAngularSpring=true
    bGravityOverrideInSimSpace=true
    bLinearSpring=true
    bOverrideAngularDamping=true
    gravityScale=0.75
    linearSpringConstant=5000
    simulationSpace=RootRelative
    alphaBoolBlend {blendOption=Linear}
    alphaScaleBiasClamp {scale=0.2}
    boundBone {boneName=dyn_pants_l_outer}
    componentAppliedLinearAccClamp=<6000 6000 6000>
    componentLinearAccScale=<0.25 0.25 0.6>
    componentLinearVelScale=<0.25 0.25 0.6>
    physicsBodyDefinitions [{
      sphereCollisionRadius=0
      boundBone {boneName=dyn_pants_l_outer}
      boxExtents=<8 25 25>
      constraintSetup {angularLimitsMax=<60 40 40>, angularLimitsMin=<-60 -40 -75>, angularTarget=<1 0 0>}
      localJointOffset=<5 0 0>
     }]
    relativeSpaceBone {boneName=root}
   }
   n4 = Anim Dynamics - Bone: dyn_pants_l_lower(ComponentPose=@n3, Alpha=1, ExternalForce="0,0,0") {
    angularDampingOverride=0.999
    angularSpringConstant=5000
    bAngularSpring=true
    bGravityOverrideInSimSpace=true
    bLinearSpring=true
    bOverrideAngularDamping=true
    gravityScale=0.75
    linearSpringConstant=5000
    simulationSpace=RootRelative
    alphaBoolBlend {blendOption=Linear}
    alphaScaleBiasClamp {scale=0.2}
    boundBone {boneName=dyn_pants_l_lower}
    componentAppliedLinearAccClamp=<6000 6000 6000>
    componentLinearAccScale=<0.25 0.25 0.6>
    componentLinearVelScale=<0.25 0.25 0.6>
    physicsBodyDefinitions [{
      sphereCollisionRadius=0
      boundBone {boneName=dyn_pants_l_lower}
      boxExtents=<8 25 25>
      constraintSetup {angularLimitsMax=<60 40 40>, angularLimitsMin=<-60 -40 -75>, angularTarget=<1 0 0>}
      localJointOffset=<5 0 0>
     }]
    relativeSpaceBone {boneName=root}
   }
   n5 = Anim Dynamics - Bone: dyn_pants_r_inner(ComponentPose=@n4, Alpha=1, ExternalForce="0,0,0") {
    angularDampingOverride=0.999
    angularSpringConstant=5000
    bAngularSpring=true
    bGravityOverrideInSimSpace=true
    bLinearSpring=true
    bOverrideAngularDamping=true
    gravityScale=0.75
    linearSpringConstant=5000
    simulationSpace=RootRelative
    alphaBoolBlend {blendOption=Linear}
    alphaScaleBiasClamp {scale=0.2}
    boundBone {boneName=dyn_pants_r_inner}
    componentAppliedLinearAccClamp=<6000 6000 6000>
    componentLinearAccScale=<0.25 0.25 0.6>
    componentLinearVelScale=<0.25 0.25 0.6>
    physicsBodyDefinitions [{
      sphereCollisionRadius=0
      boundBone {boneName=dyn_pants_r_inner}
      boxExtents=<8 25 25>
      constraintSetup {angularLimitsMax=<60 40 40>, angularLimitsMin=<-60 -40 -75>, angularTarget=<1 0 0>}
      localJointOffset=<5 0 0>
     }]
    relativeSpaceBone {boneName=root}
   }
   n6 = Anim Dynamics - Bone: dyn_pants_r_outer(ComponentPose=@n5, Alpha=1, ExternalForce="0,0,0") {
    angularDampingOverride=0.999
    angularSpringConstant=5000
    bAngularSpring=true
    bGravityOverrideInSimSpace=true
    bLinearSpring=true
    bOverrideAngularDamping=true
    gravityScale=0.75
    linearSpringConstant=5000
    simulationSpace=RootRelative
    alphaBoolBlend {blendOption=Linear}
    alphaScaleBiasClamp {scale=0.2}
    boundBone {boneName=dyn_pants_r_outer}
    componentAppliedLinearAccClamp=<6000 6000 6000>
    componentLinearAccScale=<0.25 0.25 0.6>
    componentLinearVelScale=<0.25 0.25 0.6>
    physicsBodyDefinitions [{
      sphereCollisionRadius=0
      boundBone {boneName=dyn_pants_r_outer}
      boxExtents=<8 25 25>
      constraintSetup {angularLimitsMax=<60 40 40>, angularLimitsMin=<-60 -40 -75>, angularTarget=<1 0 0>}
      localJointOffset=<5 0 0>
     }]
    relativeSpaceBone {boneName=root}
   }
   n7 = Anim Dynamics - Bone: dyn_pants_r_lower(ComponentPose=@n6, Alpha=1, ExternalForce="0,0,0") {
    angularDampingOverride=0.999
    angularSpringConstant=5000
    bAngularSpring=true
    bGravityOverrideInSimSpace=true
    bLinearSpring=true
    bOverrideAngularDamping=true
    gravityScale=0.75
    linearSpringConstant=5000
    simulationSpace=RootRelative
    alphaBoolBlend {blendOption=Linear}
    alphaScaleBiasClamp {scale=0.2}
    boundBone {boneName=dyn_pants_r_lower}
    componentAppliedLinearAccClamp=<6000 6000 6000>
    componentLinearAccScale=<0.25 0.25 0.6>
    componentLinearVelScale=<0.25 0.25 0.6>
    physicsBodyDefinitions [{
      sphereCollisionRadius=0
      boundBone {boneName=dyn_pants_r_lower}
      boxExtents=<8 25 25>
      constraintSetup {angularLimitsMax=<60 40 40>, angularLimitsMin=<-60 -40 -75>, angularTarget=<1 0 0>}
      localJointOffset=<5 0 0>
     }]
    relativeSpaceBone {boneName=root}
   }
   n8 = "Constraint - dyn_string_r_base (PPP)"!(ComponentPose=@n7, Alpha=1, ConstraintWeights_0=0.9482, ConstraintWeights_1:float=0.0328, ConstraintWeights_2:float=0.0189) -> (Pose) {
    alphaBoolBlend {blendOption=Linear}
    boneToModify {boneName=dyn_string_r_base}
    constraintSetup [{targetBone {boneName=pelvis}}, {targetBone {boneName=spine_02}}, {targetBone {boneName=spine_01}}]
    constraintWeights [1,1,1]
   }
   n9 = "Constraint - dyn_string_l_base (PPP)"!(ComponentPose=@n8, Alpha=1, ConstraintWeights_0=0.9482, ConstraintWeights_1:float=0.0328, ConstraintWeights_2:float=0.0189) -> (Pose) {
    alphaBoolBlend {blendOption=Linear}
    boneToModify {boneName=dyn_string_l_base}
    constraintSetup [{targetBone {boneName=pelvis}}, {targetBone {boneName=spine_02}}, {targetBone {boneName=spine_01}}]
    constraintWeights [1,1,1]
   }
   n10 = Get Is Sky Diving Bool() -> (Is Sky Diving Bool:bool)
   n11 = "To Float (Boolean)"(InBool=@n10)
   n12 = "Transform (Modify) Bone - Bone: dyn_string_r_base"(ComponentPose=@n9, Alpha=@n11, Translation="0,0,0", Rotation="-35.000,-95.000,-10.000", Scale="1,1,1") {
    rotationMode=BMM_Additive
    rotationSpace=BCS_BoneSpace
    translationMode=BMM_Additive
    alphaBoolBlend {blendOption=Linear}
    alphaScaleBiasClamp {bInterpResult=true}
    boneToModify {boneName=dyn_string_r_base}
    rotation=<rot -35 -95 -10>
   }
   n13 = "Transform (Modify) Bone - Bone: dyn_string_l_base"(ComponentPose=@n12, Alpha=@n11, Translation="0,0,0", Rotation="-5.000,45.000,80.000", Scale="1,1,1") {
    rotationMode=BMM_Additive
    rotationSpace=BCS_BoneSpace
    translationMode=BMM_Additive
    alphaBoolBlend {blendOption=Linear}
    alphaScaleBiasClamp {bInterpResult=true}
    boneToModify {boneName=dyn_string_l_base}
    rotation=<rot -5 45 80>
   }
   n14 = RigidBody!(ComponentPose=@n13, Alpha=SelectFloat(A=0.25, B=1, bPickA=@n10), OverrideWorldGravity:Vector=Get GravityOverride(), ExternalForce) -> (Pose) {
    bClampLinearTranslationLimitToRefPose=true
    bOverrideWorldGravity=true
    overridePhysicsAsset=&/F_MED_Sports_Fashion_Physics
    simulationSpace=BaseBoneSpace
    alphaBoolBlend {blendOption=Linear}
    alphaScaleBiasClamp {bInterpResult=true, interpSpeedDecreasing=2, interpSpeedIncreasing=80}
    baseBoneRef {boneName=pelvis}
    componentAppliedLinearAccClamp=<6000 6000 6000>
    componentLinearAccScale=<0.15 0.15 0.15>
    componentLinearVelScale=<0.15 0.15 0.15>
    overrideWorldGravity {z=-980}
   }
   n15 = Get Is Sky Diving() -> (Is Sky Diving:double)
   n16 = "Transform (Modify) Bone - Bone: dyn_pants_l_outer"(ComponentPose=@n14, Alpha=@n15, Translation="4.737,-0.000,2.467", Rotation="45,0,0", Scale="1,1,1") {
    rotationMode=BMM_Additive
    rotationSpace=BCS_BoneSpace
    translationMode=BMM_Additive
    translationSpace=BCS_BoneSpace
    alphaBoolBlend {blendOption=Linear}
    alphaScaleBiasClamp {bInterpResult=true}
    boneToModify {boneName=dyn_pants_l_outer}
    rotation {pitch=45}
    translation=<4.7369 0 2.467>
   }
   n17 = "Transform (Modify) Bone - Bone: dyn_pants_r_outer"(ComponentPose=@n16, Alpha=@n15, Translation="4.737,-0.000,2.467", Rotation="45,0,0", Scale="1,1,1") {
    rotationMode=BMM_Additive
    rotationSpace=BCS_BoneSpace
    translationMode=BMM_Additive
    translationSpace=BCS_BoneSpace
    alphaBoolBlend {blendOption=Linear}
    alphaScaleBiasClamp {bInterpResult=true}
    boneToModify {boneName=dyn_pants_r_outer}
    rotation {pitch=45}
    translation {x=4.7369, z=2.467}
   }
   Output Pose(Result=Component To Local(ComponentPose=@n17)) {layerGroup=None}
  ]}
 EventGraph {nodes [
   Event BlueprintUpdateAnimation()
   n1 = Get bIsSurfaceSwimming()
   n2 = Get bIsSkyDiving()
   > Set Is Sky Diving(Is Sky Diving:double="To Float (Boolean)"(InBool=OR Boolean(A=@n2, B=@n1))) -> (Output_Get:double)
   > Set Is Sky Diving Bool(Is Sky Diving Bool:bool=Get bIsSkydiveFloating()) -> (Output_Get:bool)
   > n3 = Gravity Override State Function DEPRECATED(
    CharacterPartAnimInstance=Self-Reference()
    GameTimeInSeconds=GetGameTimeInSeconds()
    GravityOverrideParams=Get GravityOverrideParams() -> (GravityOverrideParams:Gravity Override Params Struct Structure)
   )
   > Set GravityOverride(GravityOverride=@n3)
   n4 = GetOwnerSkeletalMeshComponent()
   Set legSwingDelta_r(legSwingDelta_r:double=MapRangeClamped(Value=GetTwistAndSwingAngleOfDeltaRotationFromRefPose(self=@n4, BoneName=thigh_r).OutSwingAngle, InRangeA=150, InRangeB=75)) -> (Output_Get:double)
   > Set legSwingDelta_l(legSwingDelta_l:double=MapRangeClamped(Value=GetTwistAndSwingAngleOfDeltaRotationFromRefPose(self=@n4, BoneName=thigh_l).OutSwingAngle, InRangeA=43, InRangeB=110)) -> (Output_Get:double)
  ]}
]
variables [
 GravityOverride {type=Vector}
 GravityOverrideParams {type=GravityOverrideParamsStruct}
 Is Sky Diving {type=number}
 Is Sky Diving Bool {type=boolean}
 legSwingDelta_l {type=number}
 legSwingDelta_r {type=number}
]
