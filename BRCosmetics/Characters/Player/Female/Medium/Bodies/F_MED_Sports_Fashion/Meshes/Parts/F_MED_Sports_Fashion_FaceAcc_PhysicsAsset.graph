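# Physics asset for the face accessory mesh of the female medium Sports Fashion body cosmetic, covering hair bangs and a hood.
# Capsule bodies sit on the head, neck and dynamic bang and hood bones.
# The head, bang base, bang parent and hood each constrain to neck_01, with angular motion free on all but the hood.
# Both bang segments attach directly to the bang parent, not in a chain, with tight swing and twist limits.
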PhysicsAsset /BRCosmetics/Characters/Player/Female/Medium/Bodies/F_MED_Sports_Fashion/Meshes/Parts/F_MED_Sports_Fashion_FaceAcc_PhysicsAsset
bodies {
 dyn_bang_01 [{shapeType=Capsule, length=1, radius=1, shapeName=dyn_bang_01_capsule, center=<1.4985 0 0>, rotation=<rot 90 0 0>}]
 dyn_bang_02 [{shapeType=Capsule, length=0.5, radius=0.5, shapeName=dyn_bang_02_capsule, center=<0.5 0 0>, rotation=<rot 90 0 0>}]
 dyn_bang_base [{shapeType=Capsule, length=0.5, radius=0.5, shapeName=dyn_bang_base_capsule, rotation=<rot 90 0 0>}]
 dyn_bang_parent [{shapeType=Capsule, length=0.5, radius=0.5, shapeName=dyn_bang_parent_capsule, rotation=<rot 90 0 0>}]
 dyn_hood [{shapeType=Capsule, length=1, radius=1, shapeName=dyn_hood_capsule, center=<4 0 0>, rotation=<rot 90 0 0>}]
 head [{shapeType=Capsule, length=3.8452, radius=5.6837, rotation=<rot 90 0 0>}]
 neck_01 [{
   shapeType=Capsule
   length=6.3026
   radius=2.8902
   shapeName=neck_01_capsule
   center=<2.1333 1.511 -0.3147>
   rotation=<rot -1.2832 -90 -2.8948>
  }]
}
constraints [
 {bone1Name=dyn_bang_01, bone2Name=dyn_bang_parent, swing1LimitDegrees=18, swing2LimitDegrees=10, twistLimitDegrees=2}
 {bone1Name=dyn_bang_02, bone2Name=dyn_bang_parent, swing1LimitDegrees=10, swing2LimitDegrees=10, twistLimitDegrees=2}
 {
  bone1Name=head
  bone2Name=neck_01
  swing1LimitDegrees=45
  swing1Motion=Free
  swing2LimitDegrees=45
  swing2Motion=Free
  twistLimitDegrees=45
  twistMotion=Free
 }
 {
  bone1Name=dyn_bang_base
  bone2Name=neck_01
  swing1LimitDegrees=45
  swing1Motion=Free
  swing2LimitDegrees=45
  swing2Motion=Free
  twistLimitDegrees=45
  twistMotion=Free
 }
 {
  bone1Name=dyn_bang_parent
  bone2Name=neck_01
  swing1LimitDegrees=45
  swing1Motion=Free
  swing2LimitDegrees=45
  swing2Motion=Free
  twistLimitDegrees=45
  twistMotion=Free
 }
 {bone1Name=dyn_hood, bone2Name=neck_01, swing1LimitDegrees=45, swing2LimitDegrees=45, twistLimitDegrees=45}
]
