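# Physics asset for the face accessory strap part of the medium female Sports Fashion player body.
# Five single-capsule bodies approximate the upper body on the head, neck_02, spine_05 and both clavicle bones.
# It defines no constraints and no bodies for strap bones, so it looks like a collision proxy for strap simulation rather than a ragdoll setup.
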PhysicsAsset /Game/Characters/Player/Female/Medium/Bodies/F_MED_Sports_Fashion/Meshes/Parts/F_MED_Sports_Fashion_FaceAcc_Strap_PhysicsAsset
bodies {
 clavicle_l [{shapeType=Capsule, length=18.5097, radius=7.8679, center=<11.138 2.0247 -0.5592>, rotation=<rot 50.3498 -97.9517 -96.1384>}]
 clavicle_r [{shapeType=Capsule, length=18.8011, radius=7.865, center=<-10.0728 -1.8599 2.1106>, rotation=<rot 50.6175 -90.9186 -90.7101>}]
 head [{shapeType=Capsule, length=2.8524, radius=9.8297, center=<-3.7242 -2.0504 0>, rotation=<rot 90 0 0>}]
 neck_02 [{shapeType=Capsule, length=9.7654, radius=8.5819, center=<-6.8413 -0.0001 0>, rotation=<rot 90 0 0>}]
 spine_05 [{shapeType=Capsule, length=11.4563, radius=11.553, center=<-3.7355 -2.2491 0.181>, rotation=<rot 90 0 0>}]
}
constraints []
