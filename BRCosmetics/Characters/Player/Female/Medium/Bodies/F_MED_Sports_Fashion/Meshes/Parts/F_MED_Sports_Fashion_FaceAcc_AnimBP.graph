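# Animation Blueprint for the face accessory of the female medium Sports Fashion body, adding secondary physics motion.
# It copies the attached parent mesh's pose, runs RigidBody simulation relative to spine_05, then has an additive dyn_hood roll node keyed to bRigidBodyEnabled.
# Each update maps head yaw relative to spine_05 into head_rot_back.
# That value feeds only a branch of bone offsets and strap RigidBody that is not connected to the output pose.
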AnimBlueprint /BRCosmetics/Characters/Player/Female/Medium/Bodies/F_MED_Sports_Fashion/Meshes/Parts/F_MED_Sports_Fashion_FaceAcc_AnimBP
target_skeleton=&/F_MED_Sports_Fashion_FaceAcc_Skeleton
graphs [
 AnimGraph {nodes [
   n1 = Get head_rot_back() -> (head_rot_back:double)
   n2 = "Transform (Modify) Bone"(Alpha=@n1, Translation="0,0,-4", Rotation="40,0.000,40", Scale="1,1,1") {
    rotationMode=BMM_Additive
    rotationSpace=BCS_BoneSpace
    alphaBoolBlend {blendOption=Linear}
    rotation {pitch=40, roll=40}
    translation {z=-4}
    boneToModify=~
   }
   n3 = "Transform (Modify) Bone"(ComponentPose=@n2, Alpha=@n1, Translation="0,0,-3", Rotation="40,0,0", Scale="1,1,1") {
    rotationMode=BMM_Additive
    rotationSpace=BCS_BoneSpace
    alphaBoolBlend {blendOption=Linear}
    rotation {pitch=40}
    translation {z=-3}
    boneToModify=~
   }
   RigidBody(ComponentPose=@n3) {
    bClampLinearTranslationLimitToRefPose=true
    bOverrideWorldGravity=true
    overridePhysicsAsset=/Game/Characters/Player/Female/Medium/Bodies/F_MED_Sports_Fashion/Meshes/Parts/F_MED_Sports_Fashion_FaceAcc_Strap_PhysicsAsset
    simulationSpace=BaseBoneSpace
    alphaBoolBlend {blendOption=Linear}
    alphaScaleBiasClamp {bInterpResult=true, interpSpeedDecreasing=4, interpSpeedIncreasing=4}
    baseBoneRef {boneName=spine_05}
    componentLinearAccScale=<0.1 0.1 0.1>
    componentLinearVelScale=<0.1 0.1 0.1>
    overrideWorldGravity {z=-980}
   }
   n4 = RigidBody(ComponentPose=Local To Component(LocalPose=Copy Pose From Mesh() {bUseAttachedParent=true})) {
    bClampLinearTranslationLimitToRefPose=true
    bOverrideWorldGravity=true
    overridePhysicsAsset=&/F_MED_Sports_Fashion_FaceAcc_PhysicsAsset
    simulationSpace=BaseBoneSpace
    alphaBoolBlend {blendOption=Linear}
    alphaScaleBiasClamp {interpSpeedDecreasing=4, interpSpeedIncreasing=4}
    baseBoneRef {boneName=spine_05}
    componentAppliedLinearAccClamp=<6000 6000 6000>
    componentLinearAccScale=<0.25 0.25 0.25>
    componentLinearVelScale=<0.25 0.25 0.25>
    overrideWorldGravity {z=-980}
   }
   n5 = "Transform (Modify) Bone - Bone: dyn_hood"!(ComponentPose=@n4, bAlphaBoolEnabled:bool=Get bRigidBodyEnabled(), Translation="0,0,0", Rotation="0,0,45", Scale="1,1,1") -> (Pose) {alphaInputType=Bool, rotationMode=BMM_Additive, alphaBoolBlend {blendOption=Linear}, boneToModify {boneName=dyn_hood}, rotation {roll=45}}
   Output Pose(Result=Component To Local(ComponentPose=@n5)) {layerGroup=None}
  ]}
 EventGraph {nodes [
   TryGetPawnOwner()
   Event BlueprintUpdateAnimation()
   n1 = BreakTransform(InTransform=GetDeltaTransformFromRefPose(self=GetOwnerSkeletalMeshComponent(), BoneName=head, BaseName=spine_05))
   Set head_rot_back(head_rot_back:double=MapRangeClamped(Value=BreakRotator(InRot=@n1.Rotation).Yaw, InRangeA=0, InRangeB=-48)) -> (Output_Get:double)
  ]}
]
variables [head_rot_back {type=number}]
