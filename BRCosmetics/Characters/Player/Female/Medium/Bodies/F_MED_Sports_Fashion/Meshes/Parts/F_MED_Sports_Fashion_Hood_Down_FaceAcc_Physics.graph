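# Physics asset for the hood-down face accessory part of the female medium Sports Fashion body, covering secondary motion of hair, bangs, hood and a hat-adjust bone.
# Capsule bodies cover those dyn_ bones plus head, neck, upper spine and right clavicle.
# Short chains hang from base bones with limited swing and tight twist, with back hair swinging widest on one axis.
# Intermediate dyn_hair_00 links join some bases to dyn_hair_base with free angular motion.
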PhysicsAsset /BRCosmetics/Characters/Player/Female/Medium/Bodies/F_MED_Sports_Fashion/Meshes/Parts/F_MED_Sports_Fashion_Hood_Down_FaceAcc_Physics
bodies {
 clavicle_r [{shapeType=Capsule, length=7.9014, radius=2.9545, center=<-3.8023 -3.269 -4.1322>, rotation=<rot -21.6953 28.0028 -64.1889>}]
 dyn_bangs_02 [{shapeType=Capsule, length=0.75, radius=0.5, shapeName=dyn_bangs_02_capsule, center=<0.9368 0 0>, rotation=<rot 90 0 0>}]
 dyn_bangs_r_01 [{shapeType=Capsule, length=0.5, radius=0.5, shapeName=dyn_bangs_r_01_capsule}]
 dyn_bangs_r_03 [{shapeType=Capsule, length=0.5, radius=1, shapeName=dyn_bangs_r_03_capsule, center=<1.0531 0 0>, rotation=<rot 90 0 0>}]
 dyn_bangs_r_04 [{shapeType=Capsule, length=0.75, radius=0.75, shapeName=dyn_bangs_r_04_capsule, center=<1.0503 0 0>, rotation=<rot 90 0 0>}]
 dyn_hair_00_bk_a [{shapeType=Capsule, length=1, radius=1, shapeName=dyn_hair_00_bk_a_capsule}]
 dyn_hair_00_bk_b [{shapeType=Capsule, length=1, radius=1, shapeName=dyn_hair_00_bk_b_capsule}]
 dyn_hair_00_bk_base_a [{shapeType=Capsule, length=1, radius=1, shapeName=dyn_hair_00_bk_base_a_capsule}]
 dyn_hair_00_bk_base_b [{shapeType=Capsule, length=1, radius=1, shapeName=dyn_hair_00_bk_base_b_capsule}]
 dyn_hair_00_bk_base_c [{shapeType=Capsule, length=1, radius=1, shapeName=dyn_hair_00_bk_base_c_capsule}]
 dyn_hair_00_bk_c [{shapeType=Capsule, length=1, radius=1, shapeName=dyn_hair_00_bk_c_capsule}]
 dyn_hair_00_l_base_c [{shapeType=Capsule, length=1, radius=1, shapeName=dyn_hair_00_l_base_c_capsule}]
 dyn_hair_00_l_c [{shapeType=Capsule, length=1, radius=1, shapeName=dyn_hair_00_l_c_capsule}]
 dyn_hair_00_r_base_c [{shapeType=Capsule, length=1, radius=1, shapeName=dyn_hair_00_r_base_c_capsule}]
 dyn_hair_00_r_c [{shapeType=Capsule, length=1, radius=1, shapeName=dyn_hair_00_r_c_capsule}]
 dyn_hair_base [{shapeType=Capsule, length=0, radius=0.5}]
 dyn_hair_bk_a_01 [{shapeType=Capsule, length=0.75, radius=1, shapeName=dyn_hair_bk_a_01_capsule, center=<1.1328 0 0>, rotation=<rot 90 0 0>}]
 dyn_hair_bk_a_02 [{shapeType=Capsule, length=1, radius=1, shapeName=dyn_hair_bk_a_02_capsule, center=<1.433 0 0>, rotation=<rot 90 0 0>}]
 dyn_hair_bk_b_01 [{shapeType=Capsule, length=0.5, radius=1, shapeName=dyn_hair_bk_b_01_capsule, center=<1.4184 0 0>, rotation=<rot 90 0 0>}]
 dyn_hair_bk_base_a [{shapeType=Capsule, length=0.5, radius=0.5, shapeName=dyn_hair_bk_base_a_capsule}]
 dyn_hair_bk_base_b [{shapeType=Capsule, length=0.5, radius=0.5, shapeName=dyn_hair_bk_base_b_capsule}]
 dyn_hair_bk_base_c [{shapeType=Capsule, length=0.5, radius=0.5, shapeName=dyn_hair_bk_base_c_capsule}]
 dyn_hair_bk_c_01 [{shapeType=Capsule, length=1.25, radius=1.25, shapeName=dyn_hair_bk_c_01_capsule, center=<2.0816 0 0>, rotation=<rot 90 0 0>}]
 dyn_hair_bk_parent_a [{shapeType=Capsule, length=2, radius=1.5, shapeName=dyn_hair_bk_parent_a_capsule, center=<2.5739 0 0>, rotation=<rot 90 0 0>}]
 dyn_hair_bk_parent_b [{shapeType=Capsule, length=2.5, radius=1.25, shapeName=dyn_hair_bk_parent_b_capsule, center=<2.2225 0 0>, rotation=<rot 90 0 0>}]
 dyn_hair_bk_parent_c [{shapeType=Capsule, length=1.75, radius=1, shapeName=dyn_hair_bk_parent_c_capsule, center=<2.731 0 -0.521>, rotation=<rot 90 0 0>}]
 dyn_hair_l_a_01 [{shapeType=Capsule, length=1, radius=1, shapeName=dyn_hair_l_a_01_capsule, center=<1.25 0 0>, rotation=<rot 90 0 0>}]
 dyn_hair_l_b_01 [{shapeType=Capsule, length=1, radius=1, shapeName=dyn_hair_l_b_01_capsule, center=<1 0 0>, rotation=<rot 90 0 0>}]
 dyn_hair_l_b_02 [{shapeType=Capsule, length=1, radius=0.75, shapeName=dyn_hair_l_b_02_capsule, center=<0.8 0 0>, rotation=<rot 90 0 0>}]
 dyn_hair_l_base_a [{shapeType=Capsule, length=0.5, radius=0.5, shapeName=dyn_hair_l_base_a_capsule}]
 dyn_hair_l_base_b [{shapeType=Capsule, length=0.5, radius=0.5, shapeName=dyn_hair_l_base_b_capsule}]
 dyn_hair_l_base_c [{shapeType=Capsule, length=0.5, radius=0.5, shapeName=dyn_hair_l_base_c_capsule}]
 dyn_hair_l_c_01 [{shapeType=Capsule, length=1, radius=1, shapeName=dyn_hair_l_c_01_capsule, center=<1.2 0 0>, rotation=<rot 90 0 0>}]
 dyn_hair_l_c_02 [{shapeType=Capsule, length=1, radius=1, shapeName=dyn_hair_l_c_02_capsule, center=<1.3569 0 0>, rotation=<rot 90 0 0>}]
 dyn_hair_l_parent_a [{shapeType=Capsule, length=1, radius=1, shapeName=dyn_hair_l_parent_a_capsule, center=<1.2 0 0>, rotation=<rot 90 0 0>}]
 dyn_hair_l_parent_b [{shapeType=Capsule, length=1, radius=1.4, shapeName=dyn_hair_l_parent_b_capsule, center=<2 0 0>, rotation=<rot 90 0 0>}]
 dyn_hair_l_parent_c [{shapeType=Capsule, length=1, radius=1.25, shapeName=dyn_hair_l_parent_c_capsule, center=<1.8526 0 0>, rotation=<rot 90 0 0>}]
 dyn_hair_r_a_01 [{shapeType=Capsule, length=0.75, radius=0.5, shapeName=dyn_hair_r_a_01_capsule, center=<0.8 0 0>, rotation=<rot 90 0 0>}]
 dyn_hair_r_a_02 [{shapeType=Capsule, length=1, radius=1, shapeName=dyn_hair_r_a_02_capsule, center=<1.6754 0 0>, rotation=<rot 90 0 0>}]
 dyn_hair_r_a_03 [{shapeType=Capsule, length=1, radius=1, shapeName=dyn_hair_r_a_03_capsule, center=<1.3489 0 0>, rotation=<rot 90 0 0>}]
 dyn_hair_r_b_01 [{shapeType=Capsule, length=1, radius=1, shapeName=dyn_hair_r_b_01_capsule, center=<1.3105 0 0>, rotation=<rot 90 0 0>}]
 dyn_hair_r_base_a [{shapeType=Capsule, length=0.5, radius=0.5, shapeName=dyn_hair_r_base_a_capsule}]
 dyn_hair_r_base_b [{shapeType=Capsule, length=0.5, radius=0.5, shapeName=dyn_hair_r_base_b_capsule}]
 dyn_hair_r_base_c [{shapeType=Capsule, length=0.5, radius=0.5, shapeName=dyn_hair_r_base_c_capsule}]
 dyn_hair_r_c_01 [{shapeType=Capsule, length=1, radius=1, shapeName=dyn_hair_r_c_01_capsule, center=<1.4586 0 0>, rotation=<rot 90 0 0>}]
 dyn_hair_r_c_02 [{shapeType=Capsule, length=1, radius=1, shapeName=dyn_hair_r_c_02_capsule, center=<1.5995 0 0>, rotation=<rot 90 0 0>}]
 dyn_hair_r_parent_a [{shapeType=Capsule, length=1, radius=1.5, shapeName=dyn_hair_r_parent_a_capsule, center=<1.6632 0 0>, rotation=<rot 90 0 0>}]
 dyn_hair_r_parent_b [{shapeType=Capsule, length=1, radius=1.5, shapeName=dyn_hair_r_parent_b_capsule, center=<1.6398 0 0>, rotation=<rot 90 0 0>}]
 dyn_hair_r_parent_c [{shapeType=Capsule, length=1.75, radius=1.5, shapeName=dyn_hair_r_parent_c_capsule, center=<2.4475 0 0>, rotation=<rot 90 0 0>}]
 dyn_hatadjust_base [{shapeType=Capsule, length=0.5, radius=0.5, shapeName=dyn_hatadjust_base_capsule}]
 dyn_hatadjust_parent [{shapeType=Capsule, length=1, radius=1, shapeName=dyn_hatadjust_parent_capsule, center=<1.3567 0 0>, rotation=<rot 90 0 0>}]
 dyn_hood_base [{shapeType=Capsule, length=0, radius=0.5, center=<-0.4097 2.7501 -0.7303>, rotation=<rot 0 0 91.6825>}]
 dyn_hood_parent [{shapeType=Capsule, length=2.5, radius=2.25, shapeName=dyn_hood_parent_capsule, center=<3.6431 0 0>, rotation=<rot 90 0 0>}]
 head [{shapeType=Capsule, length=7.972, radius=5.1895, center=<0.1559 -0.8332 0>, rotation=<rot 90 0 -10.5978>}]
 neck_02 [{shapeType=Capsule, length=7.722, radius=3.1895, center=<-2.2511 -1.2836 0>, rotation=<rot 90 0 -10.5978>}]
 spine_05 [{shapeType=Capsule, length=22, radius=8, center=<2.4976 0 0>}]
}
constraints [
 {bone1Name=dyn_hood_parent, bone2Name=dyn_hood_base, swing1LimitDegrees=30, swing2LimitDegrees=40, twistLimitDegrees=1}
 {bone1Name=dyn_hatadjust_parent, bone2Name=dyn_hatadjust_base, swing1LimitDegrees=2, swing2LimitDegrees=10, twistLimitDegrees=1}
 {bone1Name=dyn_bangs_r_03, bone2Name=dyn_bangs_r_01, swing1LimitDegrees=10, swing2LimitDegrees=10, twistLimitDegrees=3}
 {bone1Name=dyn_bangs_r_04, bone2Name=dyn_bangs_r_03, swing1LimitDegrees=10, swing2LimitDegrees=10, twistLimitDegrees=3}
 {bone1Name=dyn_bangs_02, bone2Name=dyn_bangs_r_01, swing1LimitDegrees=10, swing2LimitDegrees=10, twistLimitDegrees=3}
 {bone1Name=dyn_hair_r_parent_a, bone2Name=dyn_hair_r_base_a, swing1LimitDegrees=5, swing2LimitDegrees=5, twistLimitDegrees=3}
 {bone1Name=dyn_hair_r_a_01, bone2Name=dyn_hair_r_parent_a, swing1LimitDegrees=10, swing2LimitDegrees=10, twistLimitDegrees=3}
 {bone1Name=dyn_hair_r_a_02, bone2Name=dyn_hair_r_parent_a, swing1LimitDegrees=5, swing2LimitDegrees=5, twistLimitDegrees=3}
 {bone1Name=dyn_hair_r_a_03, bone2Name=dyn_hair_r_a_02, swing1LimitDegrees=5, swing2LimitDegrees=10, twistLimitDegrees=3}
 {bone1Name=dyn_hair_r_parent_b, bone2Name=dyn_hair_r_base_b, swing1LimitDegrees=10, swing2LimitDegrees=15, twistLimitDegrees=3}
 {bone1Name=dyn_hair_r_b_01, bone2Name=dyn_hair_r_parent_b, swing1LimitDegrees=10, swing2LimitDegrees=10, twistLimitDegrees=3}
 {bone1Name=dyn_hair_r_parent_c, bone2Name=dyn_hair_r_base_c, swing1LimitDegrees=10, swing2LimitDegrees=15, twistLimitDegrees=3}
 {bone1Name=dyn_hair_r_c_02, bone2Name=dyn_hair_r_parent_c, swing1LimitDegrees=10, swing2LimitDegrees=10, twistLimitDegrees=3}
 {bone1Name=dyn_hair_r_c_01, bone2Name=dyn_hair_r_parent_c, swing1LimitDegrees=10, swing2LimitDegrees=10, twistLimitDegrees=3}
 {bone1Name=dyn_hair_bk_parent_a, bone2Name=dyn_hair_bk_base_a, swing1LimitDegrees=10, swing2LimitDegrees=40, twistLimitDegrees=3}
 {bone1Name=dyn_hair_bk_a_01, bone2Name=dyn_hair_bk_parent_a, swing1LimitDegrees=10, swing2LimitDegrees=10, twistLimitDegrees=3}
 {bone1Name=dyn_hair_bk_a_02, bone2Name=dyn_hair_bk_parent_a, swing1LimitDegrees=10, swing2LimitDegrees=15, twistLimitDegrees=3}
 {bone1Name=dyn_hair_bk_parent_b, bone2Name=dyn_hair_bk_base_b, swing1LimitDegrees=10, swing2LimitDegrees=60, twistLimitDegrees=3}
 {bone1Name=dyn_hair_bk_b_01, bone2Name=dyn_hair_bk_parent_b, swing1LimitDegrees=10, swing2LimitDegrees=15, twistLimitDegrees=3}
 {bone1Name=dyn_hair_bk_parent_c, bone2Name=dyn_hair_bk_base_c, swing1LimitDegrees=5, swing2LimitDegrees=50, twistLimitDegrees=3}
 {bone1Name=dyn_hair_bk_c_01, bone2Name=dyn_hair_bk_parent_c, swing1LimitDegrees=5, swing2LimitDegrees=15, twistLimitDegrees=3}
 {bone1Name=dyn_hair_l_parent_c, bone2Name=dyn_hair_l_base_c, swing1LimitDegrees=5, swing2LimitDegrees=5, twistLimitDegrees=3}
 {bone1Name=dyn_hair_l_c_01, bone2Name=dyn_hair_l_parent_c, swing1LimitDegrees=10, swing2LimitDegrees=10, twistLimitDegrees=3}
 {bone1Name=dyn_hair_l_c_02, bone2Name=dyn_hair_l_parent_c, swing1LimitDegrees=10, swing2LimitDegrees=10, twistLimitDegrees=3}
 {bone1Name=dyn_hair_l_parent_b, bone2Name=dyn_hair_l_base_b, swing1LimitDegrees=10, swing2LimitDegrees=10, twistLimitDegrees=3}
 {bone1Name=dyn_hair_l_b_01, bone2Name=dyn_hair_l_parent_b, swing1LimitDegrees=10, swing2LimitDegrees=10, twistLimitDegrees=3}
 {bone1Name=dyn_hair_l_b_02, bone2Name=dyn_hair_l_b_01, swing1LimitDegrees=10, swing2LimitDegrees=10, twistLimitDegrees=3}
 {bone1Name=dyn_hair_l_parent_a, bone2Name=dyn_hair_l_base_a, swing1LimitDegrees=10, swing2LimitDegrees=10, twistLimitDegrees=3}
 {bone1Name=dyn_hair_l_a_01, bone2Name=dyn_hair_l_parent_a, swing1LimitDegrees=10, swing2LimitDegrees=10, twistLimitDegrees=3}
 {
  bone1Name=dyn_hair_r_base_c
  bone2Name=dyn_hair_00_r_c
  swing1LimitDegrees=45
  swing1Motion=Free
  swing2LimitDegrees=45
  swing2Motion=Free
  twistLimitDegrees=45
  twistMotion=Free
 }
 {
  bone1Name=dyn_hair_00_r_c
  bone2Name=dyn_hair_00_r_base_c
  swing1LimitDegrees=45
  swing1Motion=Free
  swing2LimitDegrees=45
  swing2Motion=Free
  twistLimitDegrees=45
  twistMotion=Free
 }
 {
  bone1Name=dyn_hair_00_r_base_c
  bone2Name=dyn_hair_base
  swing1LimitDegrees=45
  swing1Motion=Free
  swing2LimitDegrees=45
  swing2Motion=Free
  twistLimitDegrees=45
  twistMotion=Free
 }
 {
  bone1Name=dyn_hair_l_base_c
  bone2Name=dyn_hair_00_l_c
  swing1LimitDegrees=45
  swing1Motion=Free
  swing2LimitDegrees=45
  swing2Motion=Free
  twistLimitDegrees=45
  twistMotion=Free
 }
 {
  bone1Name=dyn_hair_00_l_base_c
  bone2Name=dyn_hair_base
  swing1LimitDegrees=45
  swing1Motion=Free
  swing2LimitDegrees=45
  swing2Motion=Free
  twistLimitDegrees=45
  twistMotion=Free
 }
 {
  bone1Name=dyn_hair_00_l_c
  bone2Name=dyn_hair_00_l_base_c
  swing1LimitDegrees=45
  swing1Motion=Free
  swing2LimitDegrees=45
  swing2Motion=Free
  twistLimitDegrees=45
  twistMotion=Free
 }
 {
  bone1Name=dyn_hair_bk_base_c
  bone2Name=dyn_hair_00_bk_c
  swing1LimitDegrees=45
  swing1Motion=Free
  swing2LimitDegrees=45
  swing2Motion=Free
  twistLimitDegrees=45
  twistMotion=Free
 }
 {
  bone1Name=dyn_hair_00_bk_base_c
  bone2Name=dyn_hair_base
  swing1LimitDegrees=45
  swing1Motion=Free
  swing2LimitDegrees=45
  swing2Motion=Free
  twistLimitDegrees=45
  twistMotion=Free
 }
 {
  bone1Name=dyn_hair_00_bk_c
  bone2Name=dyn_hair_00_bk_base_c
  swing1LimitDegrees=45
  swing1Motion=Free
  swing2LimitDegrees=45
  swing2Motion=Free
  twistLimitDegrees=45
  twistMotion=Free
 }
 {
  bone1Name=dyn_hair_bk_base_b
  bone2Name=dyn_hair_00_bk_b
  swing1LimitDegrees=45
  swing1Motion=Free
  swing2LimitDegrees=45
  swing2Motion=Free
  twistLimitDegrees=45
  twistMotion=Free
 }
 {
  bone1Name=dyn_hair_00_bk_base_b
  bone2Name=dyn_hair_base
  swing1LimitDegrees=45
  swing1Motion=Free
  swing2LimitDegrees=45
  swing2Motion=Free
  twistLimitDegrees=45
  twistMotion=Free
 }
 {
  bone1Name=dyn_hair_00_bk_b
  bone2Name=dyn_hair_00_bk_base_b
  swing1LimitDegrees=45
  swing1Motion=Free
  swing2LimitDegrees=45
  swing2Motion=Free
  twistLimitDegrees=45
  twistMotion=Free
 }
 {
  bone1Name=dyn_hair_bk_base_a
  bone2Name=dyn_hair_00_bk_a
  swing1LimitDegrees=45
  swing1Motion=Free
  swing2LimitDegrees=45
  swing2Motion=Free
  twistLimitDegrees=45
  twistMotion=Free
 }
 {
  bone1Name=dyn_hair_00_bk_base_a
  bone2Name=dyn_hair_base
  swing1LimitDegrees=45
  swing1Motion=Free
  swing2LimitDegrees=45
  swing2Motion=Free
  twistLimitDegrees=45
  twistMotion=Free
 }
 {
  bone1Name=dyn_hair_00_bk_a
  bone2Name=dyn_hair_00_bk_base_a
  swing1LimitDegrees=45
  swing1Motion=Free
  swing2LimitDegrees=45
  swing2Motion=Free
  twistLimitDegrees=45
  twistMotion=Free
 }
]
