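# Pre-deform Control Rig for the female medium Sports Fashion hood-down face accessory that offsets hair base bones based on body pose.
# Spherical pose readers run in sequence, each followed by a transform modify.
# Head relative to pelvis drives back-hair rotations from variables; each clavicle relative to neck_01 drives its side's hair, remapped to quarter strength.
# The head-to-neck branch has zero, unconnected weight, so it has no effect.
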ControlRigBlueprint /Game/Characters/Player/Female/Medium/Bodies/F_MED_Sports_Fashion/Meshes/Parts/F_MED_Sports_Fashion_Hood_Down_FaceAcc_PreDeform_CtrlRig
graphs [
 ContainedGraph {nodes [
   n1 = RigVMFunctionEntryNode()
   SequenceAggregate = &RigVMFunction_Sequence(ExecuteContext=@n1.ExecuteContext)
   SequenceAggregate_1 = &RigVMFunction_Sequence(ExecuteContext=@SequenceAggregate.B)
   SequenceAggregate_2 = &RigVMFunction_Sequence(ExecuteContext=@SequenceAggregate_1.B)
   RigVMFunctionReturnNode(A=@SequenceAggregate.A, B=@SequenceAggregate_1.A, C=@SequenceAggregate_2.A, D=@SequenceAggregate_2.B)
  ]}
 RigVMFunctionLibrary {links [], nodes []}
 RigVMModel {nodes [
   RigUnit_BeginExecution = &RigUnit_BeginExecution()
   SequenceAggregate = RigVMAggregateNode(ExecuteContext=@RigUnit_BeginExecution.ExecutePin)
   &RigUnit_SphericalPoseReader(
    ActiveRegionScaleFactors=(PositiveWidth=1,NegativeWidth=1,PositiveHeight=1,NegativeHeight=1)
    ActiveRegionSize=0.015
    Debug=(bDrawDebug=True,DebugScale=25,DebugSegments=20,DebugThickness=0.25)
    DriverAxis=(X=1,Y=0,Z=0)
    DriverItem=(Type=Bone,Name="head")
    FalloffRegionScaleFactors=(PositiveWidth=1,NegativeWidth=1,PositiveHeight=1,NegativeHeight=1)
    FalloffSize=0.06
    OptionalParentItem=(Type=Bone,Name="neck_01")
    RotationOffset=(X=90,Y=0,Z=-103)
    ExecutePin=@SequenceAggregate.A
   )
   n1 = RigVMVariableNode() {variable=hair_bk_transform}
   > &RigUnit_ModifyTransforms(
    ItemToModify="((Item=(Type=Bone,Name=\"dyn_hair_00_bk_base_c\"),Transform=(Rotation=(X=0,Y=0,Z=0,W=1),Translation=(X=0,Y=0,Z=0),Scale3D=(X=1,Y=1,Z=1))),(Item=(Type=Bone,Name=\"dyn_hair_00_bk_base_a\"),Transform=(Rotation=(X=0,Y=0,Z=0,W=1),Translation=(X=0,Y=0,Z=0),Scale3D=( ...(truncated)"
    Weight=0
    ItemToModify.0.Transform=@n1.Value
    ItemToModify.1.Transform=@n1.Value
    ItemToModify.2.Transform=@n1.Value
    ItemToModify.3.Transform=@n1.Value
   )
   n2 = &RigUnit_SphericalPoseReader(
    ActiveRegionScaleFactors=(PositiveWidth=2,NegativeWidth=0.3,PositiveHeight=2,NegativeHeight=2)
    ActiveRegionSize=0.15
    Debug=(bDrawDebug=True,DebugScale=25,DebugSegments=20,DebugThickness=0.25)
    DriverAxis=(X=1,Y=0,Z=0)
    DriverItem=(Type=Bone,Name="head")
    FalloffRegionScaleFactors=(PositiveWidth=-0.5,NegativeWidth=1,PositiveHeight=-0.5,NegativeHeight=-0.5)
    FalloffSize=0.15
    OptionalParentItem=(Type=Bone,Name="pelvis")
    RotationOffset=(X=90,Y=0,Z=-135)
    ExecutePin=@SequenceAggregate.B
   )
   n3 = RigVMVariableNode() {variable=test}
   n4 = RigVMVariableNode() {variable=hair_bk_secondary}
   > &RigUnit_ModifyTransforms(
    ItemToModify="((Item=(Type=Bone,Name=\"dyn_hair_00_bk_base_c\"),Transform=(Rotation=(X=0,Y=0,Z=0,W=1),Translation=(X=0,Y=0,Z=0),Scale3D=(X=1,Y=1,Z=1))),(Item=(Type=Bone,Name=\"dyn_hair_00_bk_base_b\"),Transform=(Rotation=(X=0,Y=0,Z=0,W=1),Translation=(X=0,Y=0,Z=0),Scale3D=( ...(truncated)"
    Weight=@n2.OutputParam
    ItemToModify.0.Transform=@n3.Value
    ItemToModify.2.Transform=@n3.Value
    ItemToModify.1.Transform=@n4.Value
    ItemToModify.3.Transform=@n4.Value
   )
   n5 = &RigUnit_SphericalPoseReader(
    ActiveRegionScaleFactors=(PositiveWidth=1,NegativeWidth=1,PositiveHeight=1,NegativeHeight=1)
    ActiveRegionSize=0.05
    Debug=(bDrawDebug=true,DebugScale=25,DebugSegments=20,DebugThickness=0.25)
    DriverAxis=(X=1,Y=0,Z=0)
    DriverItem=(Type=Bone,Name="clavicle_l")
    FalloffRegionScaleFactors=(PositiveWidth=1,NegativeWidth=1,PositiveHeight=1,NegativeHeight=1)
    FalloffSize=0.1
    OptionalParentItem=(Type=Bone,Name="neck_01")
    RotationOffset=(X=60,Y=0,Z=-90)
    ExecutePin=@SequenceAggregate.C
   )
   Remap = &RigVMFunction_MathDoubleRemap(SourceMaximum=1, SourceMinimum=0, TargetMaximum=0.25, TargetMinimum=0, bClamp=False, Value=@n5.OutputParam)
   > &RigUnit_ModifyTransforms(
    ItemToModify="((Item=(Type=Bone,Name=\"dyn_hair_l_base_a\"),Transform=(Rotation=(X=0,Y=-0.9239,Z=0,W=-0.3827),Translation=(X=0,Y=0,Z=0),Scale3D=(X=1,Y=1,Z=1))),(Item=(Type=Bone,Name=\"dyn_hair_l_base_b\"),Transform=(Rotation=(X=0,Y=-0.9239,Z=0,W=-0.3827),Translation=(X=0,Y= ...(truncated)"
    Weight=@Remap.Result
   )
   n6 = &RigUnit_SphericalPoseReader(
    ActiveRegionScaleFactors=(PositiveWidth=1,NegativeWidth=1,PositiveHeight=1,NegativeHeight=1)
    ActiveRegionSize=0.05
    Debug=(bDrawDebug=True,DebugScale=25,DebugSegments=20,DebugThickness=0.25)
    DriverAxis=(X=-1,Y=0,Z=0)
    DriverItem=(Type=Bone,Name="clavicle_r")
    FalloffRegionScaleFactors=(PositiveWidth=1,NegativeWidth=1,PositiveHeight=1,NegativeHeight=1)
    FalloffSize=0.1
    OptionalParentItem=(Type=Bone,Name="neck_01")
    RotationOffset=(X=120,Y=0,Z=90)
    ExecutePin=@SequenceAggregate.D
   )
   > &RigUnit_ModifyTransforms(
    ItemToModify="((Item=(Type=Bone,Name=\"dyn_hair_r_base_a\"),Transform=(Rotation=(X=0,Y=-0.9239,Z=0,W=-0.3827),Translation=(X=0,Y=0,Z=0),Scale3D=(X=1,Y=1,Z=1))),(Item=(Type=Bone,Name=\"dyn_hair_r_base_b\"),Transform=(Rotation=(X=0,Y=-0.9239,Z=0,W=-0.3827),Translation=(X=0,Y= ...(truncated)"
    Weight=&RigVMFunction_MathDoubleRemap(SourceMaximum=1, SourceMinimum=0, TargetMaximum=0.25, TargetMinimum=0, bClamp=False, Value=@n6.OutputParam).Result
   )
   RigVMVariableNode() {variable=test}
  ]}
]
hierarchy [
 root {loc=<0 0 0>, rot=<0 0 0 1>}
 pelvis {parent=root, loc=<0 -0.5301 98.0571>, rot=<0.0127 0.707 -0.0127 -0.707>}
 spine_01 {parent=pelvis, loc=<6.1208 -0.9328 0>, rot=<0 0 0.034 -0.9994>}
 spine_02 {parent=spine_01, loc=<6.5996 -0.2095 0>, rot=<0 0 0.0238 -0.9997>}
 spine_03 {parent=spine_02, loc=<6.4914 -0.1237 0>, rot=<0 0 0.0149 -0.9999>}
 spine_04 {parent=spine_03, loc=<6.9598 0.0794 0>, rot=<0 0 -0.0132 -0.9999>}
 spine_05 {parent=spine_04, loc=<6.9987 0 0>, rot=<0 0 -0.0015 -1>}
 clavicle_l {parent=spine_05, loc=<5.4831 -0.1053 -3.2946>, rot=<0.0465 -0.7289 -0.009 -0.6829>}
 clavicle_r {parent=spine_05, loc=<5.4831 -0.1053 3.2946>, rot=<0.7289 0.0465 -0.6829 0.009>}
 neck_01 {parent=spine_05, loc=<11.3138 0.1529 0>, rot=<0 0 -0.2003 -0.9797>}
 neck_02 {parent=neck_01, loc=<4.5627 0 0>, rot=<0 0 0.0504 -0.9987>}
 head {parent=neck_02, loc=<5.3914 0 0>, rot=<0 0 0.115 -0.9934>}
 faceAttach {parent=head, loc=<-4.8249 -0.2336 -0.0373>, rot=<0.0082 -0.7071 -0.0082 -0.7071>, scale=<0.89 0.89 0.89>}
 L_brow_mid {parent=faceAttach, loc=<3.4273 9.5342 11.0749>, rot=<-0.0001 -0.0043 0.2553 -0.9669>}
 L_brow_outer {parent=faceAttach, loc=<5.8594 6.8791 11.3043>, rot=<0.1218 0.1494 0.4055 -0.8936>}
 L_cheek_inner {parent=faceAttach, loc=<4.8674 7.6378 4.121>, rot=<-0.0262 -0.157 0.3316 -0.9299>}
 R_brow_mid {parent=faceAttach, loc=<-3.5018 9.5251 11.0795>, rot=<-0.0001 0.0043 -0.2553 -0.9669>}
 R_brow_outer {parent=faceAttach, loc=<-5.9409 6.8521 11.2901>, rot=<0.0662 -0.0795 -0.4247 -0.8994>}
 R_cheek_inner {parent=faceAttach, loc=<-4.9138 7.6164 4.1145>, rot=<-0.0266 0.1581 -0.3315 -0.9297>}
 C_jaw {parent=faceAttach, loc=<-0.1119 1.508 5.4586>, rot=<0 0 0 1>}
 dyn_hatadjust_base {parent=head, loc=<6.3969 -7.3014 -2.8398>, rot=<0.407 -0.3411 -0.6208 -0.5767>}
 dyn_hatadjust_parent {parent=dyn_hatadjust_base, loc=<0 0 0>, rot=<0 0 0 1>}
 dyn_hair_base {parent=head, loc=<9.2499 -0.5634 -0.0722>, rot=<0.0041 -1 0 0>}
 dyn_bangs_r_01 {parent=dyn_hair_base, loc=<3.4632 7.9808 -5.1702>, rot=<-0.1677 -0.2582 -0.0181 -0.9513>}
 dyn_bangs_r_03 {parent=dyn_bangs_r_01, loc=<3.3162 -0.4739 0.3304>, rot=<-0.0592 0.181 0.0005 -0.9817>}
 dyn_bangs_r_04 {parent=dyn_bangs_r_03, loc=<2.124 0.1041 0.2037>, rot=<0.0251 0.1087 -0.0529 -0.9923>}
 dyn_bangs_02 {parent=dyn_bangs_r_01, loc=<1.0969 1.6367 0.9431>, rot=<-0.4037 0.1936 -0.0486 -0.8929>}
 dyn_hair_r_base_a {parent=dyn_hair_base, loc=<4.912 5.1551 -6.7566>, rot=<-0.1371 -0.0594 0.1632 -0.9752>}
 dyn_hair_r_parent_a {parent=dyn_hair_r_base_a, loc=<0 0 0>, rot=<0 0 0 1>}
 dyn_hair_r_a_01 {parent=dyn_hair_r_parent_a, loc=<2.6846 0.0396 -0.5436>, rot=<-0.107 -0.0564 0.4189 -0.8999>}
 dyn_hair_r_a_02 {parent=dyn_hair_r_parent_a, loc=<4.4714 -0.0649 0.6843>, rot=<-0.048 0.0191 0.037 -0.998>}
 dyn_hair_r_a_03 {parent=dyn_hair_r_a_02, loc=<3.0787 -0.2259 -0.357>, rot=<0.1174 -0.3157 -0.1198 -0.9339>}
 dyn_hair_r_base_b {parent=dyn_hair_base, loc=<8.9717 -0.1698 -7.6563>, rot=<0 -0.1564 0 -0.9877>}
 dyn_hair_r_parent_b {parent=dyn_hair_r_base_b, loc=<0 0 0>, rot=<0 0 0 1>}
 dyn_hair_r_b_01 {parent=dyn_hair_r_parent_b, loc=<4.3109 0.3995 -0.399>, rot=<-0.037 -0.3663 -0.1009 -0.9243>}
 dyn_hair_l_base_b {parent=dyn_hair_base, loc=<8.4082 0.4566 7.3443>, rot=<0.9919 -0.0253 0.1192 -0.0346>}
 dyn_hair_l_parent_b {parent=dyn_hair_l_base_b, loc=<0 0 0>, rot=<0 0 0 1>}
 dyn_hair_l_b_01 {parent=dyn_hair_l_parent_b, loc=<3.8448 -0.2715 0.1538>, rot=<0.1098 -0.1081 0.0619 -0.9861>}
 dyn_hair_l_b_02 {parent=dyn_hair_l_b_01, loc=<2.4613 -0.0643 -0.2>, rot=<0.0259 -0.2755 0.0234 -0.9607>}
 dyn_hair_l_base_a {parent=dyn_hair_base, loc=<9.3692 3.799 6.4732>, rot=<0.9794 -0.0982 0.0358 0.1727>}
 dyn_hair_l_parent_a {parent=dyn_hair_l_base_a, loc=<0 0 0>, rot=<0 0 0 1>}
 dyn_hair_l_a_01 {parent=dyn_hair_l_parent_a, loc=<3.1822 0.2065 -0.0273>, rot=<-0.1931 -0.2356 0.0468 -0.9513>}
 dyn_hair_00_l_base_c {parent=dyn_hair_base, loc=<4.3653 -3.3951 5.9035>, rot=<0.942 -0.0204 0.1551 -0.297>}
 dyn_hair_00_l_c {parent=dyn_hair_00_l_base_c, loc=<0 0 0>, rot=<0 0 0 1>}
 dyn_hair_l_base_c {parent=dyn_hair_00_l_c, loc=<4.1022 0 0>, rot=<0 0 0 1>}
 dyn_hair_l_parent_c {parent=dyn_hair_l_base_c, loc=<0 0 0>, rot=<0 0 0 1>}
 dyn_hair_l_c_01 {parent=dyn_hair_l_parent_c, loc=<4.743 -2.1214 1.1543>, rot=<0.2654 -0.2253 0.1348 -0.9277>}
 dyn_hair_l_c_02 {parent=dyn_hair_l_parent_c, loc=<4.7382 2.4008 0.182>, rot=<-0.0842 -0.4751 -0.0531 -0.8743>}
 dyn_hair_00_bk_base_c {parent=dyn_hair_base, loc=<5.0751 -5.4351 0.1935>, rot=<0.7002 -0.0984 0.0984 -0.7002>}
 dyn_hair_00_bk_c {parent=dyn_hair_00_bk_base_c, loc=<0 0 0>, rot=<0 0 0 1>}
 dyn_hair_bk_base_c {parent=dyn_hair_00_bk_c, loc=<4.1022 0 0>, rot=<0 0 0 1>}
 dyn_hair_bk_parent_c {parent=dyn_hair_bk_base_c, loc=<0 0 0>, rot=<0 0 0 1>}
 dyn_hair_bk_c_01 {parent=dyn_hair_bk_parent_c, loc=<5.1476 -0.3636 -1.1705>, rot=<-0.0733 -0.5973 -0.0973 -0.7927>}
 dyn_hair_00_bk_base_b {parent=dyn_hair_base, loc=<4.6948 -5.0846 3.3662>, rot=<0.8586 -0.0446 0.1227 -0.4957>}
 dyn_hair_00_bk_b {parent=dyn_hair_00_bk_base_b, loc=<0 0 0>, rot=<0 0 0 1>}
 dyn_hair_bk_base_b {parent=dyn_hair_00_bk_b, loc=<4.1022 0 0>, rot=<0 0 0 1>}
 dyn_hair_bk_parent_b {parent=dyn_hair_bk_base_b, loc=<0 0 0>, rot=<0 0 0 1>}
 dyn_hair_bk_b_01 {parent=dyn_hair_bk_parent_b, loc=<5.4894 1.5188 -1.6857>, rot=<-0.1424 -0.5935 -0.2205 -0.7608>}
 dyn_hair_00_bk_base_a {parent=dyn_hair_base, loc=<5.1732 -4.3299 -3.4971>, rot=<0.5256 -0.1877 0.0895 -0.825>}
 dyn_hair_00_bk_a {parent=dyn_hair_00_bk_base_a, loc=<0 0 0>, rot=<0 0 0 1>}
 dyn_hair_bk_base_a {parent=dyn_hair_00_bk_a, loc=<4.1022 0 0>, rot=<0 0 0 1>}
 dyn_hair_bk_parent_a {parent=dyn_hair_bk_base_a, loc=<0 0 0>, rot=<0 0 0 1>}
 dyn_hair_bk_a_01 {parent=dyn_hair_bk_parent_a, loc=<5.3955 2.0111 -0.6665>, rot=<-0.2003 -0.5256 -0.2025 -0.8017>}
 dyn_hair_bk_a_02 {parent=dyn_hair_bk_parent_a, loc=<5.5539 -1.3174 0.1362>, rot=<0.0958 -0.531 0.2223 -0.812>}
 dyn_hair_00_r_base_c {parent=dyn_hair_base, loc=<5.3229 -2.5918 -5.591>, rot=<0.0686 -0.1776 0.041 -0.9809>}
 dyn_hair_00_r_c {parent=dyn_hair_00_r_base_c, loc=<0 0 0>, rot=<0 0 0 1>}
 dyn_hair_r_base_c {parent=dyn_hair_00_r_c, loc=<4.1022 0 0>, rot=<0 0 0 1>}
 dyn_hair_r_parent_c {parent=dyn_hair_r_base_c, loc=<0 0 0>, rot=<0 0 0 1>}
 dyn_hair_r_c_01 {parent=dyn_hair_r_parent_c, loc=<4.4818 1.3924 -0.7314>, rot=<-0.0517 -0.4862 -0.1591 -0.8577>}
 dyn_hair_r_c_02 {parent=dyn_hair_r_parent_c, loc=<4.7192 -1.555 -0.4605>, rot=<0.1773 -0.415 0.1465 -0.8803>}
 dyn_hood_base {parent=spine_05, loc=<6.5244 -13.2772 -0.2892>, rot=<0.0284 -0.7065 -0.7065 0.0284>}
 dyn_hood_parent {parent=dyn_hood_base, loc=<-0.0002 0 0.0001>, rot=<0 0 0 1>}
]
variables [
 hair_bk_secondary {type=FTransform, default=(Rotation=(X=0,Y=0.2164,Z=0,W=0.9763),Translation=(X=0,Y=0,Z=0),Scale3D=(X=1,Y=1,Z=1))}
 hair_bk_transform {type=FTransform, default=(Rotation=(X=0,Y=0.2588,Z=0,W=0.9659),Translation=(X=0,Y=0,Z=0),Scale3D=(X=1,Y=1,Z=1))}
 test {type=FTransform, default=(Rotation=(X=0,Y=0.5373,Z=0,W=0.8434),Translation=(X=0,Y=0,Z=0),Scale3D=(X=1,Y=1,Z=1))}
]
