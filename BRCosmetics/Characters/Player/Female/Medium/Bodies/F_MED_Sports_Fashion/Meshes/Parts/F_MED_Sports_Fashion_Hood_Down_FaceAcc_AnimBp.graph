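# Animation Blueprint for the hood-down face-accessory part of the female medium Sports Fashion body, animating its hood and hair bones.
# It copies the parent mesh pose, then applies head_rot_back-weighted bone offsets, front-end and pre-deform Control Rigs, spine_05-based RigidBody physics, an emote hair tilt and a final Control Rig.
# Smoothed emote state halves rig weights during emotes; the front end zeroes them and drops physics to its clamped minimum.
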AnimBlueprint /BRCosmetics/Characters/Player/Female/Medium/Bodies/F_MED_Sports_Fashion/Meshes/Parts/F_MED_Sports_Fashion_Hood_Down_FaceAcc_AnimBp
target_skeleton=&/F_MED_Sports_Fashion_Hood_Down_FaceAcc_Skeleton
graphs [
 AnimGraph {nodes [
   n1 = Get head_rot_back() -> (head_rot_back:double)
   n2 = "Transform (Modify) Bone"(Alpha=@n1, Translation="0,0,-4", Rotation="40,0.000,40", Scale="1,1,1") {
    rotationMode=BMM_Additive
    rotationSpace=BCS_BoneSpace
    alphaBoolBlend {blendOption=Linear}
    rotation {pitch=40, roll=40}
    translation {z=-4}
    boneToModify=~
   }
   "Transform (Modify) Bone"(ComponentPose=@n2, Alpha=@n1, Translation="0,0,-3", Rotation="40,0,0", Scale="1,1,1") {
    rotationMode=BMM_Additive
    rotationSpace=BCS_BoneSpace
    alphaBoolBlend {blendOption=Linear}
    rotation {pitch=40}
    translation {z=-3}
    boneToModify=~
   }
   RigidBody() {
    bClampLinearTranslationLimitToRefPose=true
    bOverrideWorldGravity=true
    overridePhysicsAsset=/Game/Characters/Player/Female/Medium/Bodies/F_MED_Sports_Fashion/Meshes/Parts/F_MED_Sports_Fashion_FaceAcc_Strap_PhysicsAsset
    simulationSpace=BaseBoneSpace
    alphaBoolBlend {blendOption=Linear}
    alphaScaleBiasClamp {bInterpResult=true, interpSpeedDecreasing=4, interpSpeedIncreasing=4}
    baseBoneRef {boneName=spine_05}
    componentLinearAccScale=<0.1 0.1 0.1>
    componentLinearVelScale=<0.1 0.1 0.1>
    overrideWorldGravity {z=-980}
   }
   n3 = Get head_rot_back()
   n4 = "Transform (Modify) Bone - Bone: dyn_hood_base"(
    ComponentPose=Local To Component(LocalPose=Copy Pose From Mesh() {bUseAttachedParent=true})
    Alpha=@n3
    Translation="5,0,0"
    Rotation="0,0,0"
    Scale="4,4,4"
   ) {
    rotationMode=BMM_Additive
    rotationSpace=BCS_BoneSpace
    scaleMode=BMM_Additive
    scaleSpace=BCS_BoneSpace
    translationMode=BMM_Additive
    translationSpace=BCS_BoneSpace
    alphaBoolBlend {blendOption=Linear}
    boneToModify {boneName=dyn_hood_base}
    scale=<4 4 4>
    translation {x=5}
   }
   n5 = "Transform (Modify) Bone - Bone: dyn_hair_bk_c_01"(ComponentPose=@n4, Alpha=@n3, Translation="0,0,0", Rotation="0,0,0", Scale="0.25,1,1") {
    rotationMode=BMM_Additive
    rotationSpace=BCS_BoneSpace
    scaleMode=BMM_Replace
    scaleSpace=BCS_BoneSpace
    alphaBoolBlend {blendOption=Linear}
    boneToModify {boneName=dyn_hair_bk_c_01}
    scale {x=0.25}
   }
   n6 = "Transform (Modify) Bone - Bone: dyn_hair_l_base_c"(ComponentPose=@n5, Alpha=@n3, Translation="0,0,0", Rotation="-60,0,10", Scale="1,1,1") {
    rotationMode=BMM_Additive
    rotationSpace=BCS_BoneSpace
    scaleMode=BMM_Additive
    scaleSpace=BCS_BoneSpace
    alphaBoolBlend {blendOption=Linear}
    boneToModify {boneName=dyn_hair_l_base_c}
    rotation {pitch=-60, roll=10}
   }
   n7 = Control Rig!(Source=Component To Local(ComponentPose=@n6), bAlphaBoolEnabled:bool=Get bIsInFrontEnd()) -> (Pose) {
    alphaInputType=Bool
    alphaBoolBlend {blendOption=Linear}
    controlRigAssetReference {blueprintRigClass=/Game/Characters/Player/Female/Medium/Bodies/F_MED_Sports_Fashion/Meshes/Parts/F_MED_Sports_Fashion_Hood_Down_FaceAcc_FrontEnd_CtrlRig.F_MED_Sports_Fashion_Hood_Down_FaceAcc_FrontEnd_CtrlRig_C}
   }
   n8 = Control Rig!(Source=@n7, Alpha:float=Get Is Playing Emote Remaped() -> (Is Playing Emote Remaped:double)) -> (Pose) {
    alphaBoolBlend {blendOption=Linear}
    controlRigAssetReference {blueprintRigClass=/Game/Characters/Player/Female/Medium/Bodies/F_MED_Sports_Fashion/Meshes/Parts/F_MED_Sports_Fashion_Hood_Down_FaceAcc_PreDeform_CtrlRig.F_MED_Sports_Fashion_Hood_Down_FaceAcc_PreDeform_CtrlRig_C}
   }
   n9 = RigidBody(ComponentPose=Local To Component(LocalPose=@n8), Alpha=Get FrontEnd_emote() -> (FrontEnd_emote:double)) {
    bClampLinearTranslationLimitToRefPose=true
    bOverrideWorldGravity=true
    overridePhysicsAsset=&/F_MED_Sports_Fashion_Hood_Down_FaceAcc_Physics
    simulationSpace=BaseBoneSpace
    alphaBoolBlend {blendOption=Linear}
    alphaScaleBiasClamp {bClampResult=true, clampMin=0.25, interpSpeedDecreasing=4, interpSpeedIncreasing=4}
    baseBoneRef {boneName=spine_05}
    componentAppliedLinearAccClamp=<3000 3000 3000>
    componentLinearAccScale=<0.22 0.22 0.5>
    componentLinearVelScale=<0.22 0.22 0.5>
    overrideWorldGravity {z=-980}
   }
   n10 = "Transform (Modify) Bone - Bone: dyn_hair_00_bk_base_a"(
    ComponentPose=@n9
    Alpha=Get Is Playing Emote() -> (Is Playing Emote:double)
    Translation="0,0,0"
    Rotation="-13.245,0.000,0.000"
    Scale="1,1,1"
   ) {
    rotationMode=BMM_Additive
    rotationSpace=BCS_BoneSpace
    alphaBoolBlend {blendOption=Linear}
    boneToModify {boneName=dyn_hair_00_bk_base_a}
    rotation=<rot -13.245 0 0>
   }
   n11 = Control Rig(Source=Component To Local(ComponentPose=@n10), Alpha=Get Is Playing Emote Remaped()) {
    alphaBoolBlend {blendOption=Linear}
    controlRigAssetReference {blueprintRigClass=&/F_MED_Sports_Fashion_Hood_Down_FaceAcc_CtrlRig.F_MED_Sports_Fashion_Hood_Down_FaceAcc_CtrlRig_C}
   }
   Output Pose(Result=@n11) {layerGroup=None}
  ]}
 EventGraph {nodes [
   n1 = Event BlueprintUpdateAnimation()
   > n2 = Sequence!(execute) -> (then_0, then_1, then_2:Exec)
   n3 = FInterpTo(Current=Get Is Playing Emote(), Target="To Float (Boolean)"(InBool=Get bIsPlayingEmote()), DeltaTime=@n1.DeltaTimeX, InterpSpeed=2)
   n4 = Get bIsInFrontEnd()
   Set Is Playing Emote(execute=@n2.then_1, Is Playing Emote:double=SelectFloat(A=0, B=@n3, bPickA=@n4)) -> (Output_Get:double)
   > Set Is Playing Emote Remaped(Is Playing Emote Remaped:double=SelectFloat(A=0, B=Remap Float(Input=@n3, OldRange=(X=0,Y=1), NewRange=(X=1,Y=0.5)) -> (Output=~), bPickA=@n4)) -> (Output_Get:double)
   n5 = Get bIsTargeting()
   Set ADS(execute=@n2.then_2, ADS:double="To Float (Boolean)"(InBool=@n5)) -> (Output_Get:double)
   Set FrontEnd_emote(execute=@n2.then_0, FrontEnd_emote:double=SelectFloat(A=0, B=1, bPickA=@n4)) -> (Output_Get:double)
   Sequence!(execute) -> (then_0, then_1)
   n6 = BreakTransform(InTransform=GetDeltaTransformFromRefPose(self=GetOwnerSkeletalMeshComponent(), BoneName=head, BaseName=spine_05))
   > Set head_rot_back(head_rot_back:double=MapRangeClamped(Value=BreakRotator(InRot=@n6.Rotation).Yaw, InRangeA=0, InRangeB=-48)) -> (Output_Get:double)
   > n7 = Branch(Condition=@n5)
   Set ADS(execute=@n7.else, ADS=0)
   TryGetPawnOwner()
  ]}
]
variables [
 ADS {type=number}
 FrontEnd_emote {type=number}
 head_rot_back {type=number}
 Is Playing Emote {type=number}
 Is Playing Emote Remaped {type=number}
]
